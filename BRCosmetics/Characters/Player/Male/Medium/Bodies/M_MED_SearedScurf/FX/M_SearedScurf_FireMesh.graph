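# Unlit, two-sided translucent material for Niagara mesh particles, drawing yellow-green fire for the SearedScurf character cosmetic.
# Wispy noise distorts upward-or-downward panning noise that drives rounded, stepped color layers over a vertical gradient, tinted by particle color.
# Opacity combines stepped noise with a sine band warped by two counter-scrolling cell noises.
# View-angle, UV-edge and depth fades soften it, scaled by particle alpha.
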Material /BRCosmetics/Characters/Player/Male/Medium/Bodies/M_MED_SearedScurf/FX/M_SearedScurf_FireMesh
graph {
 properties {
  bAutomaticallySetUsageInEditor=false
  bUsedWithNiagaraMeshParticles=true
  blendMode=BLEND_Translucent
  shadingModel=MSM_Unlit
  twoSided=true
  usedShadingModels=Unlit
 }
 nodes [
  n1 = ParticleColor()
  n2 = NamedRerouteUsage()
  n3 = Panner(Coordinate=Add(A=TextureCoordinate() {uTiling=4, vTiling=0.5}, B=Multiply(A=@n2) {constB=0.2})) {speedY=0.8}
  n4 = TextureSample(UVs=@n3) {samplerType=SAMPLERTYPE_Grayscale, texture=&LowResBlurredNoise}
  n5 = TextureCoordinate()
  n6 = Power(Base=OneMinus(None=ComponentMask(None=@n5) {g=true})) {constExponent=1.5}
  n7 = Power(Base=@n4.R, Exp=Clamp(None=LinearInterpolate(Alpha=@n6) {constA=-1, constB=2}) {maxDefault=5})
  n8 = TextureCoordinate() {uTiling=3, vTiling=0.5}
  n9 = LinearInterpolate(
   A=TextureSample(UVs=Panner(Coordinate=@n8) {speedX=0.3, speedY=0.4}) {samplerType=SAMPLERTYPE_Grayscale, texture=/BRCosmetics/Weapons/FORT_Melee/Pickaxe_Cube_Ninja_Male/FX/T_CellNoise}.R
   B=TextureSample(UVs=Panner(Coordinate=@n8) {speedX=-0.3, speedY=0.4}) {samplerType=SAMPLERTYPE_Grayscale, texture=&T_CellNoise}.R
  )
  n10 = TextureCoordinate()
  n11 = Multiply(A=Power(Base=ComponentMask(None=@n10) {g=true}) {constExponent=10}) {constB=10}
  n12 = Add(A=TextureCoordinate(), B=Multiply(A=Multiply(A=@n9) {constB=0.5}, B=Saturate(None=@n11)))
  n13 = Power(Base=Sine(None=ComponentMask(None=@n12) {g=true}) {period=2}) {constExponent=3}
  n14 = Constant() {r=0.3}
  n15 = MaterialFunctionCall("FadeSensitivity (optional)"=Constant() {r=1}) {materialFunction=&FadeBasedOnViewAngle}
  n16 = Multiply(
   A=Round(None=Power(Base=@n7) {constExponent=0.5})
   B=Multiply(
    A=MaterialFunctionCall(In=@n13, Contrast=@n14) {materialFunction=&CheapContrast}.Result
    B=@n15.Result
   )
  )
  n17 = TextureCoordinate() {uTiling=2, vTiling=0.5}
  n18 = LinearInterpolate(Alpha=TextureSample(UVs=Panner(Coordinate=@n17) {speedY=0.2}) {samplerType=SAMPLERTYPE_Grayscale, texture=&LowResBlurredNoise}.R) {constA=0.3, constB=2}
  n19 = TextureCoordinate()
  n20 = Power(Base=OneMinus(None=ComponentMask(None=@n19) {g=true}))
  n21 = Multiply(A=Multiply(A=@n1.A, B=Multiply(A=@n16, B=@n18)), B=Saturate(None=Multiply(A=@n20) {constB=10}))
  n22 = Sine(None=ComponentMask(None=TextureCoordinate()) {r=true}) {period=2}
  n23 = TextureCoordinate()
  n24 = Multiply(A=Power(Base=ComponentMask(None=@n23) {g=true})) {constB=10}
  DepthFade(Opacity=Multiply(A=Multiply(A=@n21, B=@n22), B=Saturate(None=@n24))) {fadeDistanceDefault=5}
  MaterialFunctionCall(In=@n15.Result, Contrast=Constant() {r=1}) {materialFunction=&CheapContrast}
  n25 = TextureSample(UVs=@n3) {samplerType=SAMPLERTYPE_Grayscale, texture=&T_CellNoise}
  n26 = LinearInterpolate(Alpha=@n25.R) {constA=0.3}
  n27 = Multiply(A=LinearInterpolate(Alpha=@n25.R) {constA=0.6, constB=2}, B=LinearInterpolate(Alpha=Power(Base=@n4.R, Exp=@n6)) {constA=-0.5, constB=1.5})
  n28 = Add(
   A=Multiply(A=VectorParameter() {parameterName=CenterColor2, defaultValue=<rgba 0.4 0.3342 0.0153 1>}.RGB, B=Round(None=@n26))
   B=Multiply(A=VectorParameter() {parameterName=CenterColor, defaultValue=<rgba 0.263 0.4 0.004 1>}.RGB, B=Round(None=@n27))
  )
  n29 = ComponentMask(None=TextureCoordinate()) {g=true}
  n30 = Multiply(
   A=LinearInterpolate(
    A=VectorParameter() {parameterName=Param_2, defaultValue=<rgba 0.8333 1 0 0>}.RGB
    B=VectorParameter() {parameterName=Param_3, defaultValue=<rgba 3 4 0 1>}.RGB
    Alpha=Power(Base=@n29)
   )
   B=LinearInterpolate(
    A=VectorParameter() {defaultValue=<rgba 0.3 0.3 0 1>}.RGB
    B=VectorParameter() {parameterName=Param_1, defaultValue=<rgba 0.502 0.502 0.0118 1>}.RGB
    Alpha=@n7
   )
  )
  MaterialFunctionCall(Emissive=Multiply(A=@n1.RGB, B=Add(A=@n28, B=@n30))) {materialFunction=&MF_FX_EmissiveAdjustments}
  n31 = Panner(Coordinate=TextureCoordinate() {uTiling=2, vTiling=0.7}) {speedY=0.2}
  NamedRerouteDeclaration(None=Subtract(A=TextureSample(UVs=@n31) {samplerType=SAMPLERTYPE_Grayscale, texture=/Game/Effects/Fort_Effects/Textures/Fire/T_WispyFireNoise_Soft_M}.R) {constB=0.5}) {name=Disturb, nodeColor=<rgba 0 0.25 1 1>}
  Saturate(None=Multiply(A=@n13) {constB=10})
  TextureSample(UVs=@n3) {samplerType=SAMPLERTYPE_Grayscale, texture=&LowResBlurredNoise}
 ]
}
