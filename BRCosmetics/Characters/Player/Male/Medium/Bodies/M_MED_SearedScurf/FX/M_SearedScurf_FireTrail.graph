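# Translucent, unlit, two-sided fire trail material for the Seared Scurf character's FX, used on Niagara ribbons and mesh particles.
# World position offset combines SplineThicken width, sine waviness and a clamped, particle-driven bend.
# Opacity layers panning textures, edge and lead masks, optional dissolve and appear masks, and depth fade.
# Emissive sums two panning color layers, tinted by particle and main colors with an optional lead-to-tail gradient, while dynamic parameters drive shape, speed and dissolve.
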Material /BRCosmetics/Characters/Player/Male/Medium/Bodies/M_MED_SearedScurf/FX/M_SearedScurf_FireTrail
graph {
 properties {
  bAutomaticallySetUsageInEditor=false
  bUsedWithNiagaraMeshParticles=true
  bUsedWithNiagaraRibbons=true
  blendMode=BLEND_Translucent
  shadingModel=MSM_Unlit
  twoSided=true
  usedShadingModels=Unlit
 }
 nodes [
  n1 = Constant() {r=20}
  n2 = ScalarParameter() {defaultValue=1, parameterName=WidthScale}
  n3 = DynamicParameter() {defaultValue=<rgba 1 1 1 0>, paramNames [WidthBase,WidthTip,Offset,WPO]}
  n4 = ComponentMask(None=TextureCoordinate()) {r=true}
  n5 = Multiply(A=Power(Base=OneMinus(None=@n4)) {constExponent=1.5}) {constB=3}
  n6 = ComponentMask(None=TextureCoordinate()) {r=true}
  n7 = Saturate(None=Multiply(A=Power(Base=@n6) {constExponent=1.5}) {constB=7})
  n8 = Multiply(
   A=LinearInterpolate(A=ScalarParameter() {defaultValue=0.5, group=SplineTicken, parameterName=TailWidth}, Alpha=Saturate(None=@n5)) {constA=0.5}
   B=LinearInterpolate(B=LinearInterpolate(Alpha=@n7) {constA=0.8}, Alpha=ScalarParameter() {group=SplineTicken, parameterName=UseLeftWidthMask}) {constA=1}
  )
  n9 = Multiply(A=Multiply(A=Multiply(A=@n1, B=@n2), B=@n3.WidthBase), B=@n8)
  n10 = Constant() {r=10}
  n11 = Multiply(A=Multiply(A=Multiply(A=@n10, B=@n2), B=@n3.WidthTip), B=@n8)
  n12 = StaticBool()
  n13 = Add(A=ComponentMask(None=TextureCoordinate()) {r=true}, B=NamedRerouteUsage())
  n14 = Time()
  n15 = ScalarParameter() {defaultValue=-0.6, group=WPO, parameterName=WPOSpeed}
  n16 = Sine(None=Add(A=ComponentMask(None=@n13) {r=true}, B=Multiply(A=@n14, B=@n15) {constB=-0.4})) {period=0.6}
  n17 = Transform(None=VectorParameter() {group=WPO, parameterName=WPODirection, defaultValue=<rgba 0 0 1 1>}.RGB)
  n18 = Transform(None=VectorParameter() {group=WPO, parameterName=WPODirection2, defaultValue=<rgba 0 0 0 1>}.RGB)
  n19 = Multiply(
   A=Add(A=Multiply(A=@n16, B=@n17), B=Multiply(A=@n16, B=@n18))
   B=DynamicParameter() {defaultValue=<rgba 1 1 1 0>, paramNames [WidthBase,WidthTip,Offset,WPO]}.WPO
  )
  n20 = TextureCoordinate()
  n21 = Saturate(None=Multiply(A=ComponentMask(None=@n20) {r=true}) {constB=2})
  n22 = DynamicParameter() {parameterIndex=1, paramNames [X, Y, Z, BendIntensity]}
  n23 = MaterialFunctionCall(Float3=Subtract(
    A=MaterialFunctionCall(X=@n22.X, Y=@n22.X, Z=@n22.X) {materialFunction=&MakeFloat3}.Result
    B=ParticlePositionWS()
   )) {materialFunction=&BreakOutFloat3Components}
  n24 = ScalarParameter() {defaultValue=50, parameterName=ClampRange}
  n25 = Multiply(A=@n24) {constB=-1}
  n26 = ScalarParameter() {defaultValue=10, parameterName=ClampRange_B}
  n27 = MaterialFunctionCall(
   X=Clamp(None=@n23.R, Min=@n25, Max=@n24)
   Y=Clamp(None=@n23.G, Min=@n25, Max=@n24)
   Z=Clamp(None=@n23.B, Min=Multiply(A=@n26) {constB=-1}, Max=@n26)
  ) {materialFunction=&MakeFloat3}
  n28 = TextureCoordinate()
  n29 = Multiply(A=@n22.BendIntensity, B=Power(Base=ComponentMask(None=@n28) {r=true}, Exp=ScalarParameter() {defaultValue=3, parameterName=BendPower}))
  Add(
   A=Add(
    A=MaterialFunctionCall(WidthBase=@n9, WidthTip=@n11, Expand U or V UV Channel=@n12) {materialFunction=&SplineThicken}.WorldPositionOffset
    B=Multiply(A=@n19, B=@n21)
   )
   B=LinearInterpolate(B=LinearInterpolate(B=@n27.Result, Alpha=@n29), Alpha=ScalarParameter() {defaultValue=1, parameterName=UseInterpolateOvertime})
  )
  Constant() {r=5}
  n30 = ParticleColor()
  n31 = VectorParameter() {group=Alpha, parameterName=DissolveUV, defaultValue=<rgba 0.7 0.5 0 0>}
  n32 = Add(A=Multiply(A=TextureCoordinate(), B=ComponentMask(None=@n31.RGB) {g=true, r=true}), B=NamedRerouteUsage())
  n33 = Constant2Vector() {g=0.3}
  n34 = Constant2Vector() {r=-0.2}
  n35 = DynamicParameter() {parameterIndex=2, defaultValue=<rgba 0 1 1 0>, paramNames [Dissolve, Speed, Appear, CustomUV1Sidemask]}
  n36 = TextureSample(UVs=Panner(Coordinate=Add(A=@n32, B=@n33), Speed=Multiply(A=@n34, B=@n35.Speed)) {speedX=-0.2}) {samplerType=SAMPLERTYPE_Masks, texture=/Game/Weapons/WeaponSkins/Wraps/Textures/Mash/Mash_Noise}
  n37 = ScalarParameter() {parameterName=Dissolve mask min}
  n38 = ScalarParameter() {defaultValue=1, parameterName=Dissolve mask max}
  n39 = TextureCoordinate()
  n40 = ScalarParameter() {defaultValue=-0.6, parameterName=Offset}
  n41 = Power(Base=ComponentMask(None=Add(A=@n39, B=@n40)) {r=true}) {constExponent=0.5}
  n42 = Sine(None=ComponentMask(None=TextureCoordinate()) {g=true}) {period=2}
  n43 = Multiply(A=Saturate(None=LinearInterpolate(A=@n37, B=@n38, Alpha=@n41)), B=LinearInterpolate(Alpha=OneMinus(None=@n42)) {constA=1, constB=3})
  n44 = Saturate(None=LinearInterpolate(
    A=ScalarParameter() {defaultValue=-1, parameterName=Disappear min}
    B=ScalarParameter() {defaultValue=3, parameterName=Disappear max}
    Alpha=Power(Base=@n36.R, Exp=@n43) {constExponent=1}
   ))
  n45 = ComponentMask(None=TextureCoordinate()) {r=true}
  n46 = Multiply(
   A=Power(Base=OneMinus(None=@n45), Exp=ScalarParameter() {defaultValue=1, group=Alpha, parameterName=TailDissolveFade}) {constExponent=1}
   B=ScalarParameter() {defaultValue=3, group=Alpha, parameterName=TailDissolveEdgeIntensity}
  ) {constB=3}
  n47 = TextureSample(UVs=Panner(Coordinate=TextureCoordinate() {uTiling=2, vTiling=0.7}) {speedX=-0.1}) {samplerType=SAMPLERTYPE_Masks, texture=/Game/Effects/Fort_Effects/Textures/T_PackedNoise}
  n48 = Multiply(A=Multiply(A=ComponentMask(None=@n47.RGB) {b=true, g=true}, B=Constant2Vector() {g=0.5, r=1})) {constB=0.08}
  n49 = ScalarParameter() {parameterName=UseDisturb}
  n50 = Add(
   A=Add(A=TextureCoordinate(), B=NamedRerouteUsage())
   B=NamedRerouteDeclaration(None=LinearInterpolate(B=@n48, Alpha=@n49)) {name=Disturb, nodeColor=<rgba 1 0 0.0211 1>}
  )
  n51 = ComponentMask(None=VectorParameter() {group=Alpha, parameterName=Alpha_UVScale, defaultValue=<rgba 1.5 1 0 0>}.RGB) {g=true, r=true}
  n52 = ComponentMask(None=VectorParameter() {group=Alpha, parameterName=AlphaSpeed, defaultValue=<rgba -0.5 0 0 0>}.RGB) {g=true, r=true}
  n53 = DynamicParameter() {parameterIndex=2, defaultValue=<rgba 0 1 1 0>, paramNames [Dissolve, Speed, Appear, CustomUV1Sidemask]}
  n54 = TextureSampleParameter2D(UVs=Panner(Coordinate=Multiply(A=@n50, B=@n51), Speed=Multiply(A=@n52, B=@n53.Speed)) {speedX=-0.5}) {
   group=Alpha
   parameterName=Alpha
   sortPriority=34
   texture=&/T_SearedScurf_Trail
  }
  n55 = Multiply(A=Power(
    Base=StaticComponentMaskParameter(None=@n54.RGB) {defaultR=true, group=Alpha, parameterName=AlphaMask, sortPriority=33}
    Exp=ScalarParameter() {defaultValue=0.5, group=Alpha, parameterName=AlphaPower}
   ) {constExponent=0.3}) {constB=2}
  n56 = Multiply(A=MaterialFunctionCall(In=Saturate(None=@n55), Contrast=ScalarParameter() {defaultValue=0.5, group=Alpha, parameterName=AlphaContrast}) {materialFunction=&CheapContrast}.Result)
  n57 = LinearInterpolate(
   A=TextureSample() {samplerType=SAMPLERTYPE_Grayscale, texture=&LowResBlurredNoise}.R
   Alpha=ScalarParameter() {defaultValue=0.3, group=Alpha, parameterName=Alpha noise lerp}
  ) {constAlpha=0.3}
  n58 = Saturate(None=LinearInterpolate(
    A=ScalarParameter() {group=Alpha, parameterName=AlphaMin}
    B=ScalarParameter() {defaultValue=1, group=Alpha, parameterName=AlphaMax}
    Alpha=Multiply(A=@n56, B=@n57)
   ))
  n59 = TextureCoordinate()
  n60 = Power(Base=Sine(None=ComponentMask(None=@n59) {g=true}) {period=2}, Exp=ScalarParameter() {group=Alpha, parameterName=Alpha edge mask})
  n61 = Multiply(A=Multiply(A=LinearInterpolate(Alpha=@n44) {constA=-0.3, constB=2}, B=Saturate(None=@n46)), B=Saturate(None=Multiply(A=@n58, B=@n60)))
  n62 = TextureCoordinate() {uTiling=0.8}
  n63 = Add(A=TextureCoordinate() {uTiling=1.2}, B=NamedRerouteUsage())
  n64 = Multiply(A=TextureSample(UVs=Panner(Coordinate=@n63) {speedX=-0.2, speedY=0.5}) {samplerType=SAMPLERTYPE_Grayscale, texture=&LowResBlurredNoise}.R) {constB=-0.12}
  n65 = SphereMask(A=Add(A=Add(A=@n62, B=@n64), B=Constant2Vector() {r=-0.1}), B=Constant() {r=0.5}, Radius=Constant() {r=0.6}) {attenuationRadius=0.6, hardnessPercent=90}
  n66 = ScalarParameter() {defaultValue=1, parameterName=UseLeadMask}
  n67 = DynamicParameter() {parameterIndex=2, defaultValue=<rgba 0 1 1 0>, paramNames [Dissolve, Speed, Appear, CustomUV1Sidemask]}
  n68 = TextureCoordinate()
  n69 = ComponentMask(None=VectorParameter() {group=Alpha, parameterName=DissolveUVScale, defaultValue=<rgba 2 1 0 0>}.RGB) {g=true, r=true}
  n70 = TextureSampleParameter2D(UVs=Add(A=Multiply(A=@n68, B=@n69), B=NamedRerouteUsage())) {group=Alpha, parameterName=Dissolve_Tex, samplerType=SAMPLERTYPE_Grayscale, texture=&LowResBlurredNoise}
  n71 = MaterialFunctionCall(
   Transition Phase=@n67.Dissolve
   Height Texture=@n70.R
   Contrast=ScalarParameter() {defaultValue=5, group=Alpha, parameterName=DissolveContrast}
  ) {materialFunction=&HeightLerp}
  n72 = Multiply(
   A=Multiply(A=Multiply(A=@n30.A, B=@n61), B=LinearInterpolate(B=@n65, Alpha=@n66) {constA=1})
   B=LinearInterpolate(B=Multiply(A=@n71.Results), Alpha=ScalarParameter() {group=Alpha, parameterName=Use AnimTrailDissolve}) {constA=1}
  )
  n73 = TextureCoordinate()
  n74 = TextureCoordinate() {coordinateIndex=1}
  n75 = ScalarParameter() {group=Alpha, parameterName=UseUV1Type}
  n76 = Sine(None=ComponentMask(None=LinearInterpolate(A=@n73, B=@n74, Alpha=@n75) {constA=1}) {r=true}) {period=2}
  n77 = DynamicParameter() {parameterIndex=2, defaultValue=<rgba 0 1 1 0>, paramNames [Dissolve, Speed, Appear, CustomUV1Sidemask]}
  n78 = Multiply(
   A=ScalarParameter() {defaultValue=1.5, group=Alpha, parameterName=UV1 SideMask}
   B=LinearInterpolate(B=@n77.CustomUV1Sidemask, Alpha=ScalarParameter() {parameterName=UseCustomUV1Side}) {constA=1}
  )
  n79 = Saturate(None=Multiply(A=Power(Base=@n76, Exp=@n78) {constExponent=1.5}) {constB=2})
  n80 = TextureCoordinate()
  n81 = DynamicParameter() {parameterIndex=2, defaultValue=<rgba 0 1 1 0>, paramNames [Dissolve, Speed, Appear, CustomUV1Sidemask]}
  n82 = Add(A=OneMinus(None=ComponentMask(None=@n80) {r=true}), B=LinearInterpolate(Alpha=@n81.Appear) {constA=-1})
  n83 = Constant() {r=1}
  n84 = Saturate(None=Multiply(A=MaterialFunctionCall(In=@n82, Contrast=@n83) {materialFunction=&CheapContrast}.Result))
  n85 = ScalarParameter() {group=Alpha, parameterName=Use Appear}
  DepthFade(
   Opacity=Multiply(A=Multiply(A=@n72, B=@n79), B=LinearInterpolate(B=@n84, Alpha=@n85) {constA=1})
   FadeDistance=ScalarParameter() {defaultValue=5, parameterName=FadeDistance}
  ) {fadeDistanceDefault=5}
  LinearInterpolate()
  LinearInterpolate(Alpha=MaterialFunctionCall() {materialFunction=&DebugTimeSine}.Result) {constA=0.4, constB=0.8}
  n86 = Add(A=TextureCoordinate(), B=NamedRerouteUsage())
  n87 = ComponentMask(None=VectorParameter() {group=TexC, parameterName=TexC_UVScale, defaultValue=<rgba 1 0.7 0 0>}.RGB) {g=true, r=true}
  n88 = ComponentMask(None=VectorParameter() {group=TexC, parameterName=TexC_Speed, defaultValue=<rgba -0.5 0 0 0>}.RGB) {g=true, r=true}
  n89 = DynamicParameter() {parameterIndex=2, defaultValue=<rgba 0 1 1 0>, paramNames [Dissolve, Speed, Appear, CustomUV1Sidemask]}
  n90 = TextureSampleParameter2D(UVs=Panner(Coordinate=Multiply(A=@n86, B=@n87), Speed=Multiply(A=@n88, B=@n89.Speed)) {speedX=-0.5}) {group=TexC, parameterName=TexC, samplerType=SAMPLERTYPE_Grayscale, texture=&LowResBlurredNoise}
  n91 = LinearInterpolate(
   A=ScalarParameter() {defaultValue=0.2, group=TexC, parameterName=TexCmin}
   B=ScalarParameter() {defaultValue=1.2, group=TexC, parameterName=TexCmax}
   Alpha=@n90.R
  )
  n92 = ScalarParameter() {group=TexC, parameterName=TexCPower}
  LinearInterpolate(
   B=Multiply(A=VectorParameter() {group=TexC, parameterName=TexCColor, defaultValue=<rgba 1 1 1 0>}.RGB, B=Power(Base=@n91, Exp=@n92) {constExponent=1})
   Alpha=ScalarParameter() {group=TexC, parameterName=UseTexCNoise}
  ) {constA=1}
  n93 = VectorParameter() {group=TexB, parameterName=TexBColor1, defaultValue=<rgba 0 0 0 1>}
  n94 = VectorParameter() {group=TexB, parameterName=TexBColor2, defaultValue=<rgba 0.3373 0.5 0.0119 1>}
  n95 = Add(A=TextureCoordinate(), B=NamedRerouteUsage())
  n96 = ComponentMask(None=VectorParameter() {group=TexB, parameterName=TexB_UVScale, defaultValue=<rgba 0.7 0.5 0 0>}.RGB) {g=true, r=true}
  n97 = ComponentMask(None=VectorParameter() {group=TexB, parameterName=TexB_Speed, defaultValue=<rgba -0.3 0 0 0>}.RGB) {g=true, r=true}
  n98 = DynamicParameter() {parameterIndex=2, defaultValue=<rgba 0 1 1 0>, paramNames [Dissolve, Speed, Appear, CustomUV1Sidemask]}
  n99 = TextureSampleParameter2D(UVs=Panner(Coordinate=Multiply(A=@n95, B=@n96), Speed=Multiply(A=@n97, B=@n98.Speed)) {speedX=-0.3}) {group=TexB, parameterName=TexB, texture=/BRCosmetics/Weapons/FORT_Melee/Pickaxe_CoatCheck/FX/T_Pickaxe_CoatCheck_Noise}
  n100 = ScalarParameter() {defaultValue=1.5, group=TexB, parameterName=TexBPower}
  n101 = Saturate(None=LinearInterpolate(Alpha=Power(Base=@n99.R, Exp=@n100) {constExponent=1.5}))
  n102 = ScalarParameter() {group=TexA, parameterName=TexAmin}
  n103 = ScalarParameter() {defaultValue=1, group=TexA, parameterName=TexAmax}
  n104 = Add(A=Add(A=TextureCoordinate(), B=NamedRerouteUsage()), B=NamedRerouteUsage())
  n105 = ComponentMask(None=VectorParameter() {group=TexA, parameterName=TexA_UVScale, defaultValue=<rgba 1.5 1 0 0>}.RGB) {g=true, r=true}
  n106 = ComponentMask(None=VectorParameter() {group=TexA, parameterName=TexA_Speed, defaultValue=<rgba -0.5 0 0 0>}.RGB) {g=true, r=true}
  n107 = DynamicParameter() {parameterIndex=2, defaultValue=<rgba 0 1 1 0>, paramNames [Dissolve, Speed, Appear, CustomUV1Sidemask]}
  n108 = TextureSampleParameter2D(UVs=Panner(Coordinate=Multiply(A=@n104, B=@n105), Speed=Multiply(A=@n106, B=@n107.Speed)) {speedX=-0.5}) {
   group=TexA
   parameterName=TexA
   sortPriority=34
   texture=&T_SearedScurf_Trail
  }
  n109 = Power(Base=StaticComponentMaskParameter(None=@n108.RGB) {defaultR=true, group=TexA, parameterName=TexAMask, sortPriority=33})
  n110 = LinearInterpolate(
   A=VectorParameter() {group=TexA, parameterName=TexAColor1, defaultValue=<rgba 0.1049 0.2 0.0099 1>}.RGB
   B=VectorParameter() {group=TexA, parameterName=TexAColor2, defaultValue=<rgba 0.3542 0.5 0.0625 1>}.RGB
   Alpha=Power(Base=LinearInterpolate(A=@n102, B=@n103, Alpha=@n109), Exp=ScalarParameter() {defaultValue=1, group=TexA, parameterName=TexAPower}) {constExponent=1}
  )
  n111 = VectorParameter() {parameterName=TailColor, sortPriority=37, defaultValue=<rgba 0.8 0.8 0.8 0>}
  n112 = VectorParameter() {parameterName=LeadColor, sortPriority=36, defaultValue=<rgba 0.8514 1 0 1>}
  n113 = OneMinus(None=ComponentMask(None=TextureCoordinate()) {r=true})
  n114 = ScalarParameter() {defaultValue=3.5}
  n115 = Saturate(None=Multiply(A=Power(Base=@n113, Exp=@n114) {constExponent=1}, B=ScalarParameter() {defaultValue=3, parameterName=Param_1}) {constB=3})
  n116 = Multiply(
   A=Add(A=LinearInterpolate(A=@n93.RGB, B=@n94.RGB, Alpha=@n101), B=Multiply(B=@n110) {constA=1})
   B=LinearInterpolate(B=LinearInterpolate(A=@n111.RGB, B=@n112.RGB, Alpha=@n115), Alpha=ScalarParameter() {parameterName=UseColorMask}) {constA=1}
  )
  MaterialFunctionCall(Emissive=Multiply(A=Multiply(A=@n116, B=@n30.RGB), B=VectorParameter() {parameterName=MainColor, defaultValue=<rgba 1.1 0.925 0.4 1>}.RGB)) {materialFunction=&MF_FX_EmissiveAdjustments}
  MaterialFunctionCall() {materialFunction=&DebugTimeSine}
  Multiply()
  n117 = DynamicParameter() {defaultValue=<rgba 1 1 1 0>, paramNames [WidthBase,WidthTip,Offset,WPO]}
  n118 = VertexColor()
  NamedRerouteDeclaration(None=AppendVector(A=Multiply(A=@n117.Offset, B=@n118.R), B=Constant())) {name=Offset, nodeColor=<rgba 0.145 0 1 1>}
 ]
}
