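# Translucent, unlit, two-sided Niagara mesh-particle material for the SearedScurf character cosmetic, drawing scrolling green-to-yellow wisps.
# Panning noise blends the colors, tinted by particle color.
# Opacity combines particle alpha, a flame texture, view-angle fade, UV end fades and a near-hard caustic-noise erosion, randomizable through an Offset dynamic parameter.
# A U-weighted bend uses clamped differences between particle position and dynamic parameter X, replicated across all three axes.
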Material /BRCosmetics/Characters/Player/Male/Medium/Bodies/M_MED_SearedScurf/FX/M_SearedScurf_HighMesh
graph {
 properties {
  bAutomaticallySetUsageInEditor=false
  bUsedWithNiagaraMeshParticles=true
  blendMode=BLEND_Translucent
  shadingModel=MSM_Unlit
  twoSided=true
  usedShadingModels=Unlit
 }
 nodes [
  n1 = DynamicParameter() {parameterIndex=1, paramNames [X, Y, Z, BendIntensity]}
  n2 = MaterialFunctionCall(Float3=Subtract(
    A=MaterialFunctionCall(X=@n1.X, Y=@n1.X, Z=@n1.X) {materialFunction=&MakeFloat3}.Result
    B=ParticlePositionWS()
   )) {materialFunction=&BreakOutFloat3Components}
  n3 = ScalarParameter() {defaultValue=50, parameterName=ClampRange}
  n4 = Multiply(A=@n3) {constB=-1}
  n5 = Clamp(None=@n2.R, Min=@n4, Max=@n3)
  n6 = Clamp(None=@n2.G, Min=@n4, Max=@n3)
  n7 = ScalarParameter() {defaultValue=10, parameterName=ClampRange_B}
  n8 = Clamp(None=@n2.B, Min=Multiply(A=@n7) {constB=-1}, Max=@n7)
  n9 = Power(Base=ComponentMask(None=TextureCoordinate()) {r=true}, Exp=ScalarParameter() {defaultValue=3, parameterName=BendPower})
  LinearInterpolate(
   B=LinearInterpolate(
    B=MaterialFunctionCall(X=@n5, Y=@n6, Z=@n8) {materialFunction=&MakeFloat3}.Result
    Alpha=Multiply(A=@n1.BendIntensity, B=@n9)
   )
   Alpha=ScalarParameter() {defaultValue=1, parameterName=UseInterpolateOvertime}
  )
  n10 = ParticleColor()
  n11 = OneMinus(None=ComponentMask(None=TextureCoordinate()) {r=true})
  n12 = LinearInterpolate(Alpha=Saturate(None=Multiply(A=@n11))) {constA=0.7}
  n13 = Constant3Vector() {constant=<rgba 0.12 0.18 0 1>}
  n14 = Constant3Vector() {constant=<rgba 0.2667 0.4 0 1>}
  n15 = TextureSample(UVs=Panner(Coordinate=TextureCoordinate() {uTiling=0.5, vTiling=1.5}) {speedX=-0.4}) {samplerType=SAMPLERTYPE_Grayscale, texture=&LowResBlurredNoise}
  n16 = TextureCoordinate() {uTiling=0.7}
  n17 = AppendVector(A=Constant() {r=-0.8}, B=Constant())
  n18 = Power(Base=TextureSample(UVs=Panner(Coordinate=@n16, Speed=@n17) {speedX=-0.4}) {texture=/BRCosmetics/Weapons/FORT_Melee/Pickaxe_SoapPocket/FX/T_Pickaxe_SoapPocket_Noise}.R) {constExponent=4}
  n19 = LinearInterpolate(
   A=Multiply(A=LinearInterpolate(A=@n13, B=@n14, Alpha=@n15.R))
   B=Constant3Vector() {constant=<rgba 2.4458 2.5 0.0336 1>}
   Alpha=Multiply(A=Saturate(None=@n18), B=@n15.R)
  )
  MaterialFunctionCall(Emissive=Multiply(A=@n10.RGB, B=Multiply(A=@n12, B=@n19) {constA=1})) {materialFunction=&MF_FX_EmissiveAdjustments}
  n20 = Add(A=TextureCoordinate() {uTiling=0.6}, B=AppendVector(A=DynamicParameter() {paramNames [Offset,Param2,Param3,Param4]}.Offset, B=Constant()))
  n21 = AppendVector(A=Constant() {r=-1}, B=Constant())
  n22 = Multiply(A=TextureSampleParameter2D(UVs=Panner(Coordinate=@n20, Speed=@n21) {speedX=-0.4}) {parameterName=AlphaTex, texture=/BRCosmetics/Characters/Player/Male/Medium/Bodies/M_MED_SoapPocket/FX/Materials/T_SoapPocket_Flame}.R) {constB=10}
  n23 = ScalarParameter() {defaultValue=1, parameterName=AngleFade}
  n24 = Multiply(
   A=Multiply(A=Saturate(None=@n22))
   B=MaterialFunctionCall(
    In=MaterialFunctionCall("FadeSensitivity (optional)"=@n23) {materialFunction=&FadeBasedOnViewAngle}.Result
    Contrast=Constant() {r=5}
   ) {materialFunction=&CheapContrast}.Result
  )
  n25 = DynamicParameter() {paramNames [Offset,Param2,Param3,Param4]}
  n26 = Constant()
  n27 = Add(A=Add(A=TextureCoordinate() {uTiling=0.8}, B=AppendVector(A=@n25.Offset, B=@n26)), B=Constant2Vector() {g=0.3})
  n28 = Power(Base=TextureSample(UVs=Panner(Coordinate=@n27) {speedX=-1}) {samplerType=SAMPLERTYPE_LinearGrayscale, texture=/Game/Global/Textures/Default/Caustics/T_Nosie_Caustics_Grayscale}.R) {constExponent=1.2}
  n29 = Constant()
  n30 = Constant() {r=0.5}
  n31 = Power(Base=ComponentMask(None=TextureCoordinate()) {r=true}) {constExponent=1.3}
  n32 = Sine(None=ComponentMask(None=TextureCoordinate()) {g=true}) {period=2}
  n33 = Multiply(A=Saturate(None=LinearInterpolate(A=@n29, B=@n30, Alpha=@n31)), B=LinearInterpolate(Alpha=OneMinus(None=@n32)) {constA=1, constB=3})
  n34 = Saturate(None=LinearInterpolate(A=Constant() {r=-1.5}, B=Constant() {r=3}, Alpha=Power(Base=@n28, Exp=@n33)))
  n35 = Multiply(A=LinearInterpolate(A=Round(None=@n34), B=@n34) {constAlpha=0.1})
  n36 = ComponentMask(None=TextureCoordinate()) {r=true}
  n37 = Saturate(None=Multiply(A=OneMinus(None=@n36)) {constB=5})
  n38 = TextureCoordinate()
  n39 = Multiply(A=Power(Base=ComponentMask(None=@n38) {r=true})) {constB=5}
  Multiply(
   A=Multiply(A=Multiply(A=@n10.A, B=@n24), B=LinearInterpolate(Alpha=@n15.R) {constA=0.3})
   B=Multiply(A=Multiply(A=@n35, B=@n37), B=Saturate(None=@n39))
  )
 ]
}
